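# Translucent unlit Niagara mesh-particle material for the back-wheel effect in the Yuzucrank Pita emote.
# Samples a lerp texture with UV tiling scaled around the center.
# A dynamic parameter drives a lerp, but both inputs read the red channel, so it has no visible effect.
# The result is multiplied by particle color into an emissive adjustment function and by particle alpha into depth-faded, saturated opacity.
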Material /BRCosmetics/Animation/Game/MainPlayer/Emotes/YuzuCrank_Pita/FX/M_Emote_Yuzucrank_Pita_BackWheelLerp
graph {
 properties {
  bAutomaticallySetUsageInEditor=false
  bUsedWithNiagaraMeshParticles=true
  blendMode=BLEND_Translucent
  shadingModel=MSM_Unlit
  usedShadingModels=Unlit
 }
 nodes [
  n1 = Add(A=TextureCoordinate()) {constB=-0.5}
  n2 = AppendVector(A=ScalarParameter() {defaultValue=1, parameterName=UV Tile X}, B=ScalarParameter() {defaultValue=1, parameterName=UV Tile Y})
  n3 = TextureSampleParameter2D(UVs=Add(A=Multiply(A=@n1, B=@n2)) {constB=0.5}) {parameterName=Lerp Texture, texture=&/T_Emote_Yuzucrank_Pita_BackWheelLerp}
  n4 = LinearInterpolate(
   A=@n3.R
   B=@n3.R
   Alpha=DynamicParameter() {defaultValue=<rgba 0 1 1 1>, paramNames [LerpTexture Alpha, Param2, Param3, Param4]}.LerpTexture Alpha
  )
  n5 = ParticleColor()
  MaterialFunctionCall(Emissive=Multiply(A=@n4, B=@n5.RGB)) {materialFunction=&MF_FX_EmissiveAdjustments}
  Saturate(None=DepthFade(Opacity=Multiply(A=@n5.A, B=@n4), FadeDistance=ScalarParameter() {defaultValue=25, parameterName=Fade Distance}))
 ]
}
