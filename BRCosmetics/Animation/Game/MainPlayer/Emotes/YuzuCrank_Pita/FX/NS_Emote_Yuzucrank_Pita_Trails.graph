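# Niagara system for the YuzuCrank Pita emote, drawing glowing warm-colored trails and a rear-wheel glow on the animated skeletal mesh.
# Three AnimTrail emitters layer ribbon renderers with per-layer color and width, using time-of-day scaling on some.
# RearWheelEmissive follows skeletal mesh sample points with mesh, sprite and light renderers.
# Owner speed drives trail lifetime, drift, world/local blending and wheel glow, with front-end overrides replacing gameplay values in menus.
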NiagaraSystem /BRCosmetics/Animation/Game/MainPlayer/Emotes/YuzuCrank_Pita/FX/NS_Emote_Yuzucrank_Pita_Trails
dependencies {
 usedDataInterfaces [
  &NiagaraDataInterfaceCurve
  &NiagaraDataInterfaceColorCurve
  &NiagaraDataInterfaceSkeletalMesh
  &NiagaraDataInterfaceSocketReader
 ]
 usedDynamicInputs [&RemapRange]
 usedModules [
  &self:SystemSpawnScript.SystemScriptSource.SystemScriptGraph.NiagaraNodeAssignment_0.SetVariables_#1
  &SystemState
  &BlendSpaceHelper
  &Timeline
  &self:SystemSpawnScript.SystemScriptSource.SystemScriptGraph.NiagaraNodeAssignment_1.SetVariables_#2
  &InterpolateOverTime
 ]
}
emitters {
 FN_AnimTrail {
  emitterSpawnScript [&FN_RibbonIDOffsetUtility()]
  emitterUpdateScript [
   &EmitterState()
   &FN_SpawnAnimTrail()
  ]
  particleSpawnScript [
   &FN_InitializeAnimTrailBoneIndecies()
   &FN_InitializeAnimTrailBoneIndecies(AnimTrail Selection=1, First Bone or Socket Index=2, Second Bone or Socket Index=3) {enabled=false, module=FN_InitializeAnimTrailBoneIndecies001}
   &FN_InitializeAnimTrail(
    Color=<rgba 12 10.52 3.75 1>
    Lifetime=&FN_FrontEndFloatModifier(
     Gameplay=&Lerp_Float(
      A=0.225
      B=0.125
      Alpha=&NormalizeFloat(
       Value=&FloatFromCurve(
        FloatCurve=DataInterface() {propertyValues {Curve {keys [<at 150 0 RCIM_Constant>,<at 493 1 RCIM_Constant>]}}}
        CurveIndex=&VectorLength(Vector=$User.OwnerLocalVelocity)
       )
       Min=0
       Max=1
      )
     )
     Front End=0.125
    )
    Position Offset=<0 0 1.2>
   )
   &AddVelocity(Coordinate Space=::Simulation, Velocity=<0 -625 0>) {enabled=false}
   &AddVelocity(Velocity=&MakeVector(
     X=0
     Y=&FN_FrontEndFloatModifier(
      Gameplay=&Lerp_Float(
       A=0
       B=-625
       Alpha=&FloatFromCurve(
        FloatCurve=DataInterface() {propertyValues {Curve {keys [<at 0 0 RCIM_Constant>,<at 500 1 RCIM_Constant>]}}}
        CurveIndex=&VectorLength(Vector=$Engine.Owner.Velocity)
       )
      )
      Front End=-425
     )
     Z=0
    )) {module=AddVelocity001}
  ]
  particleUpdateScript [
   &ScaleVelocity(Velocity Scale=&MakeVector(
     X=0
     Y=&FloatFromCurve(FloatCurve=DataInterface() {propertyValues {Curve {keys [<at 0 1 RCIM_Cubic>,<at 0.15 0 RCIM_Cubic>]}}})
     Z=0
    )) {enabled=false}
   &BlendSpace("Position Blend (World <-> Local)":NiagaraFloat=&MakeCustomFloatFromBool(
     Bool=&SetBoolByFloatComparison(A=&VectorLength(Vector=$Engine.Owner.Velocity), B=500)
     True Float=1
     False Float=0
     Translation Contribution:NiagaraFloat=1
     Rotation Contribution:NiagaraFloat=0
    ))
   &SolveForcesAndVelocity(Clamp Velocity=false, Limit Acceleration=false)
   &FN_UpdateAnimTrail() {module=FN_UpdateAnimTrail001}
   &DynamicMaterialParameters(Module.Index0Float0Write=true, Module.Index0Float1Write=true, Module.Index0Float2Write=true, Module.Index0Float3Write=true)
   &ScaleColor(
    Scale Alpha=&Multiply_Float(
     A=&FloatFromCurve(
      FloatCurve=DataInterface() {propertyValues {Curve {keys [<at 0 1 RCIM_Cubic>, {interpMode=RCIM_Cubic, leaveTangent=-2.2142, time=1, value=0}]}}}
      Scale Curve=0.1
     )
     B=&FN_FrontEndFloatModifier(Gameplay=&Lerp_Float(A=0, B=1, Alpha=$System.Timeline.NormalizedTime), Front End=1)
    )
    ScaleA=true
    ScaleRGB=true
   )
  ]
  rendererClasses [&NiagaraRibbonRendererProperties]
  renderers [
   {rendererClass=&NiagaraRibbonRendererProperties}
   {rendererClass=&NiagaraRibbonRendererProperties, rendererIndex=1}
  ]
 }
 FN_AnimTrail001 {
  emitterSpawnScript [&FN_RibbonIDOffsetUtility()]
  emitterUpdateScript [
   &EmitterState()
   &FN_SpawnAnimTrail()
  ]
  particleSpawnScript [
   &FN_InitializeAnimTrailBoneIndecies()
   &FN_InitializeAnimTrailBoneIndecies(AnimTrail Selection=1, First Bone or Socket Index=2, Second Bone or Socket Index=3) {enabled=false, module=FN_InitializeAnimTrailBoneIndecies001}
   &FN_InitializeAnimTrail(
    AnimTrailWidth=1.2
    Color=<rgba 1 1 1 0.45>
    Lifetime=&FN_FrontEndFloatModifier(
     Gameplay=&Lerp_Float(
      A=1.2
      B=0.35
      Alpha=&NormalizeFloat(
       Value=&FloatFromCurve(
        FloatCurve=DataInterface() {propertyValues {Curve {keys [<at 150 0 RCIM_Constant>,<at 493 1 RCIM_Constant>]}}}
        CurveIndex=&VectorLength(Vector=$User.OwnerLocalVelocity)
       )
       Min=0
       Max=1
      )
     )
     Front End=0.35
    )
    Position Offset=<0 0 2>
   )
   &AddVelocity(Coordinate Space=::Simulation, Velocity=<0 -625 0>) {enabled=false}
   &AddVelocity(Velocity=&MakeVector(
     X=0
     Y=&FN_FrontEndFloatModifier(
      Gameplay=&Lerp_Float(
       A=0
       B=-625
       Alpha=&FloatFromCurve(
        FloatCurve=DataInterface() {propertyValues {Curve {keys [<at 0 0 RCIM_Constant>,<at 490 1 RCIM_Constant>]}}}
        CurveIndex=&VectorLength(Vector=$User.OwnerLocalVelocity)
       )
      )
      Front End=-425
     )
     Z=0
    )) {module=AddVelocity001}
   &PartitionParticles(Number Of Partitions=15) {enabled=false}
   /SetVariables(Particles.LightPos:NiagaraPosition=&AddVectorToPosition(Position=$Particles.Position, Vector=<0 0 -22.5>)) {enabled=false, isSetParameters=true}
  ]
  particleUpdateScript [
   &BlendSpace("Position Blend (World <-> Local)":NiagaraFloat=&MakeCustomFloatFromBool(
     Bool=&SetBoolByFloatComparison(A=&VectorLength(Vector=$Engine.Owner.Velocity), B=500)
     True Float=1
     False Float=0
     Translation Contribution:NiagaraFloat=1
     Rotation Contribution:NiagaraFloat=0
    ))
   &ScaleVelocity(Velocity Scale=&MakeVector(
     X=0
     Y=&FloatFromCurve(FloatCurve=DataInterface() {propertyValues {Curve {keys [<at 0 1 RCIM_Cubic>,<at 0.15 0 RCIM_Cubic>]}}})
     Z=0
    )) {enabled=false}
   &SolveForcesAndVelocity(Clamp Velocity=false, Limit Acceleration=false)
   &FN_UpdateAnimTrail() {module=FN_UpdateAnimTrail001}
   &ScaleColor(
    Scale Alpha=&Multiply_Float(
     A=&FloatFromCurve(
      FloatCurve=DataInterface() {propertyValues {Curve {keys [
          <at 0 1>
          {arriveTangent=-0.0395, interpMode=RCIM_Cubic, leaveTangent=-0.0395, tangentMode=RCTM_User, time=0.25, value=1}
          {arriveTangent=-2.2142, interpMode=RCIM_Cubic, leaveTangent=-2.2142, tangentMode=RCTM_User, time=1, value=0}
         ]}}}
      Scale Curve=0.7
     )
     B=1
    )
    ScaleA=true
    ScaleRGB=true
   )
   &ScaleRibbonWidth(Ribbon Width Scale=0.95)
   /SetVariables(
    Particles.Color2=&Multiply_LinearColorByFloat(
     Linear Color=<rgba 25 2.0414 1.0417 1>
     Float=&Multiply_Float(
      A=&FloatFromCurve(FloatCurve=DataInterface() {propertyValues {Curve {keys [
           <at 0 1>
           {arriveTangent=-0.0395, interpMode=RCIM_Cubic, leaveTangent=-0.0395, tangentMode=RCTM_User, time=0.25, value=1}
           {arriveTangent=-1.6584, interpMode=RCIM_Cubic, leaveTangent=-1.6584, tangentMode=RCTM_User, time=1, value=0}
          ]}}})
      B=1
     )
    )
    Particles.Dynamic2:Vector4f=<3 1.45 0 0>
    Particles.RibbonWidth2:NiagaraFloat=&Multiply_Float(A=$Particles.RibbonWidth, B=0.85)
   ) {isSetParameters=true}
   /SetVariables(
    Particles.Color3:LinearColor=&Multiply_LinearColorByFloat(
     Linear Color=<rgba 10 0.0676 0 1>
     Float=&Multiply_Float(
      A=&FloatFromCurve(FloatCurve=DataInterface() {propertyValues {Curve {keys [
           <at 0 1>
           {arriveTangent=-0.0395, interpMode=RCIM_Cubic, leaveTangent=-0.0395, tangentMode=RCTM_User, time=0.25, value=1}
           {arriveTangent=-1.6562, interpMode=RCIM_Cubic, leaveTangent=-1.6562, tangentMode=RCTM_User, time=1, value=0}
          ]}}})
      B=1
     )
    )
    Particles.Dynamic3:Vector4f=<1 1 0 0>
    Particles.RibbonWidth3:NiagaraFloat=&Multiply_Float(A=$Particles.RibbonWidth, B=0.875)
   ) {isSetParameters=true}
   /SetVariables(
    Particles.Color4:LinearColor=&Multiply_LinearColorByFloat(
     Linear Color=<rgba 50 0.338 0 1>
     Float=&Multiply_Float(
      A=&FloatFromCurve(
       FloatCurve=DataInterface() {propertyValues {Curve {keys [
           <at 0 1>
           {arriveTangent=-0.0395, interpMode=RCIM_Cubic, leaveTangent=-0.0395, tangentMode=RCTM_User, time=0.25, value=1}
           {arriveTangent=-1.6562, interpMode=RCIM_Cubic, leaveTangent=-1.6562, tangentMode=RCTM_User, time=1, value=0}
          ]}}}
       Scale Curve=0.2
      )
      B=1
     )
    )
    Particles.Dynamic4:Vector4f=&MakeVector4(
     X=&FloatFromCurve(FloatCurve=DataInterface() {propertyValues {Curve {keys [<at 0 1>,<at 1 0>]}}})
     Y=0
     Z=0
     W=1
    )
    Particles.RibbonWidth4:NiagaraFloat=&Multiply_Float(A=$Particles.RibbonWidth, B=1.3)
   ) {isSetParameters=true}
   /SetVariables(
    Particles.Color5:LinearColor=&Multiply_LinearColorByFloat(
     Linear Color=<rgba 25 8.9824 0 1>
     Float=&Multiply_Float(
      A=&FloatFromCurve(FloatCurve=DataInterface() {propertyValues {Curve {keys [
           <at 0 1>
           {arriveTangent=-0.0395, interpMode=RCIM_Cubic, leaveTangent=-0.0395, tangentMode=RCTM_User, time=0.25, value=1}
           {arriveTangent=-1.6562, interpMode=RCIM_Cubic, leaveTangent=-1.6562, tangentMode=RCTM_User, time=1, value=0}
          ]}}})
      B=1
     )
    )
    Particles.Dynamic5:Vector4f=&MakeVector4(
     X=&FloatFromCurve(FloatCurve=DataInterface() {propertyValues {Curve {keys [<at 0 1>,<at 1 0>]}}})
     Y=0
     Z=0
     W=1
    )
    Particles.RibbonWidth5:NiagaraFloat=&Multiply_Float(A=$Particles.RibbonWidth, B=1.25)
   ) {isSetParameters=true}
   /SetVariables(
    Particles.Color6:LinearColor=&Multiply_LinearColorByFloat(
     Linear Color=&ColorFromCurve(ColorCurve=DataInterface() {propertyValues {
        AlphaCurve {keys [<at 0 1 RCIM_Cubic>,<at 1 0 RCIM_Cubic>]}
        BlueCurve {keys [<at 0 0 RCIM_Cubic>,<at 0.1119 0>,<at 0.2248 0 RCIM_Cubic>,<at 1 0>]}
        GreenCurve {keys [
          <at 0 6.3373 RCIM_Cubic>
          <at 0.1119 0.1379>
          {arriveTangent=-0.1455, interpMode=RCIM_Cubic, leaveTangent=-0.1455, time=0.2248, value=0.0261}
          <at 1 0.0087>
         ]}
        RedCurve {keys [
          <at 0 25 RCIM_Cubic>
          <at 0.1119 17>
          {arriveTangent=-13.5118, interpMode=RCIM_Cubic, leaveTangent=-13.5118, time=0.2248, value=15}
          <at 1 5>
         ]}
       }})
     Float=&Multiply_Float(
      A=&FloatFromCurve(
       FloatCurve=DataInterface() {propertyValues {Curve {keys [<at 0 1 RCIM_Cubic>, {arriveTangent=-0.0104, interpMode=RCIM_Cubic, leaveTangent=-0.0104, tangentMode=RCTM_User, time=0.8, value=0}]}}}
       Scale Curve=&FNTimeOfDayScalar(Night Scale=1.5, Morning Scale=1.25, Day Scale=2.5, Evening Scale=1.5)
      )
      B=1
     )
    )
    Particles.Dynamic6:Vector4f=<0 0 0 0>
    Particles.RibbonWidth6:NiagaraFloat=&Multiply_Float(
     A=$Particles.RibbonWidth
     B=&FloatFromCurve(
      FloatCurve=DataInterface() {propertyValues {Curve {keys [
          <at 0 0.8 RCIM_Cubic>
          {arriveTangent=-0.2209, interpMode=RCIM_Cubic, leaveTangent=-0.2209, tangentMode=RCTM_User, time=0.275, value=1}
          <at 1 0 RCIM_Cubic>
         ]}}}
      Scale Curve=1.25
     )
    )
   ) {isSetParameters=true, module=SetVariables_#3}
   &FN_Light_Attributes(
    Color=&ColorFromCurve(ColorCurve=DataInterface() {propertyValues {
       AlphaCurve {keys [<at 0 1 RCIM_Cubic>,<at 1 0 RCIM_Cubic>]}
       BlueCurve {keys [<at 0 0 RCIM_Cubic>,<at 0.1782 0 RCIM_Cubic>]}
       GreenCurve {keys [<at 0 0.4295 RCIM_Cubic>,<at 0.1782 0.0244 RCIM_Cubic>]}
       RedCurve {keys [<at 0 1 RCIM_Cubic>,<at 0.1782 1 RCIM_Cubic>]}
      }})
    Create Color Attribute=true
    Create Exponent Attribute=true
    Create Radius Attribute=true
    Enable Particle Light Enable=true
    Exponent=4
    Light Enabled:NiagaraBool=&Boolean_AndOperation(
     Bool1=/Niagara/DynamicInputs/Helpers/MaskBool(
      Mask Source:NiagaraInt32=1
      Mask Parameter:NiagaraInt32=$Particles.PartitionParticles.Partition
      Mask:NiagaraInt32=0
      Masked Bool:NiagaraBool=true
      Passthrough Bool:NiagaraBool=false
     )
     Bool2=&SetBoolByFloatComparison(A=$System.Timeline.NormalizedTime, B=0)
    )
    Radius=150
    Scale Color=&FloatFromCurve(
     FloatCurve=DataInterface() {propertyValues {Curve {
        defaultValue=0
        keys [{interpMode=RCIM_Cubic, tangentMode=RCTM_User, time=0, value=1}, {interpMode=RCIM_Cubic, tangentMode=RCTM_User, time=1, value=0}]
       }}}
     Scale Curve=7
    )
   ) {enabled=false}
  ]
  rendererClasses [&NiagaraRibbonRendererProperties]
  renderers [
   {rendererClass=&NiagaraRibbonRendererProperties}
   {rendererClass=&NiagaraRibbonRendererProperties, rendererIndex=1}
   {rendererClass=&NiagaraRibbonRendererProperties, rendererIndex=2}
   {rendererClass=&NiagaraRibbonRendererProperties, rendererIndex=3}
   {rendererClass=&NiagaraRibbonRendererProperties, rendererIndex=4}
   {rendererClass=&NiagaraRibbonRendererProperties, rendererIndex=5}
  ]
 }
 FN_AnimTrail002 {
  emitterSpawnScript [&FN_RibbonIDOffsetUtility()]
  emitterUpdateScript [
   &EmitterState()
   &FN_SpawnAnimTrail()
  ]
  particleSpawnScript [
   &FN_InitializeAnimTrailBoneIndecies()
   &FN_InitializeAnimTrailBoneIndecies(AnimTrail Selection=1, First Bone or Socket Index=2, Second Bone or Socket Index=3) {enabled=false, module=FN_InitializeAnimTrailBoneIndecies001}
   &FN_InitializeAnimTrail(
    AnimTrailWidth=1.2
    Color=<rgba 1 1 1 0.45>
    Lifetime=&FN_FrontEndFloatModifier(
     Gameplay=&Lerp_Float(
      A=1.2
      B=0.35
      Alpha=&NormalizeFloat(
       Value=&FloatFromCurve(
        FloatCurve=DataInterface() {propertyValues {Curve {keys [<at 150 0 RCIM_Constant>,<at 493 1 RCIM_Constant>]}}}
        CurveIndex=&VectorLength(Vector=$User.OwnerLocalVelocity)
       )
       Min=0
       Max=1
      )
     )
     Front End=0.35
    )
    Position Offset=<0 0 8>
   )
   &AddVelocity(Coordinate Space=::Simulation, Velocity=<0 -625 0>) {enabled=false}
   &AddVelocity(Velocity=&MakeVector(
     X=0
     Y=&FN_FrontEndFloatModifier(
      Gameplay=&Lerp_Float(
       A=0
       B=-625
       Alpha=&FloatFromCurve(
        FloatCurve=DataInterface() {propertyValues {Curve {keys [<at 0 0 RCIM_Constant>,<at 500 1 RCIM_Constant>]}}}
        CurveIndex=&VectorLength(Vector=$Engine.Owner.Velocity)
       )
      )
      Front End=-425
     )
     Z=0
    )) {module=AddVelocity001}
  ]
  particleUpdateScript [
   &ScaleVelocity(Velocity Scale=&MakeVector(
     X=0
     Y=&FloatFromCurve(FloatCurve=DataInterface() {propertyValues {Curve {keys [<at 0 1 RCIM_Cubic>,<at 0.15 0 RCIM_Cubic>]}}})
     Z=0
    )) {enabled=false}
   &BlendSpace("Position Blend (World <-> Local)":NiagaraFloat=&MakeCustomFloatFromBool(
     Bool=&SetBoolByFloatComparison(A=&VectorLength(Vector=$Engine.Owner.Velocity), B=500)
     True Float=1
     False Float=0
     Translation Contribution:NiagaraFloat=1
     Rotation Contribution:NiagaraFloat=0
    ))
   &SolveForcesAndVelocity(Clamp Velocity=false, Limit Acceleration=false)
   &FN_UpdateAnimTrail() {module=FN_UpdateAnimTrail001}
   /SetVariables(
    Particles.Color6:LinearColor=&Multiply_LinearColorByFloat(
     Linear Color=<rgba 12 0 0 1>
     Float=&Multiply_Float(
      A=&FloatFromCurve(
       FloatCurve=DataInterface() {propertyValues {Curve {keys [<at 0 1 RCIM_Cubic>, {arriveTangent=-2.5155, interpMode=RCIM_Cubic, leaveTangent=-2.5155, tangentMode=RCTM_User, time=1, value=0}]}}}
       Scale Curve=&FNTimeOfDayScalar(Night Scale=1.5, Morning Scale=1.25, Day Scale=2.5, Evening Scale=1.5)
      )
      B=&FN_FrontEndFloatModifier(Gameplay=$System.Timeline.NormalizedTime, Front End=1)
     )
    )
    Particles.Dynamic6:Vector4f=<0 0 0 0>
    Particles.RibbonWidth6:NiagaraFloat=&Multiply_Float(A=$Particles.RibbonWidth, B=1.5)
   ) {isSetParameters=true, module=SetVariables_#3}
  ]
  rendererClasses [&NiagaraRibbonRendererProperties]
  renderers [
   {rendererClass=&NiagaraRibbonRendererProperties}
   {rendererClass=&NiagaraRibbonRendererProperties, rendererIndex=1}
   {rendererClass=&NiagaraRibbonRendererProperties, rendererIndex=2}
  ]
 }
 RearWheelEmissive {
  emitterUpdateScript [
   &EmitterState(Life Cycle Mode=::NewEnumerator1, Loop Behavior=::NewEnumerator1, Loop Duration Mode=::NewEnumerator1, UseLoopDelay=false)
   &SpawnBurst_Instantaneous(Use Spawn Probability=false, UseLoopCountLimit=false)
   &SpawnRate(
    SpawnGroup=1
    SpawnRate=&MakeCustomFloatFromBool(
     Bool=&SetBoolByFloatComparison(A=&VectorLength(Vector=$User.OwnerLocalVelocity), B=150)
     True Float=8
     False Float=0
    )
    Use Spawn Probability=false
   )
  ]
  particleSpawnScript [
   /SetVariables(Particles.VisibilityTag=&SpawnGroup()) {isSetParameters=true}
   &InitializeParticle(
    Color=<rgba 25 0.0699 0 1>
    Lifetime=&MergeHelper_FloatSwitch(
     0=7.645
     1=&RandomRangeFloat(Minimum=1, Maximum=1.7)
     2=?
     3=?
     Index=&SpawnGroup()
     Num=::NewEnumerator0
    )
    Mesh Scale=<1 1.1 1.025>
    Mesh Scale Mode=::NewEnumerator4
    Sprite Size Mode=::NewEnumerator1
    Uniform Sprite Size Max=50
    Uniform Sprite Size Min=35
    UsePositionOffset=false
    Write Material Random=true
   )
   &AddVelocity(
    Module.Use Falloff Distance=false
    Override Seed=false
    Velocity Mode=::NewEnumerator1
    Velocity Speed=&MergeHelper_FloatSwitch(0=0, 1=25, 2=?, 3=?, Index=&SpawnGroup(), Num=::NewEnumerator0)
   )
  ]
  particleUpdateScript [
   &ParticleState(Kill Particles When Lifetime Has Elapsed=false, Let Infinitely Lived Particles Die When Emitter Deactivates=true)
   &SkeletalMeshLocation(
    MaskSpawnGroup=false
    Mesh Sampling Type=::NewEnumerator5
    Sampled Position Offset=<0 37 -20>
    Skeletal Mesh=$System.Skeletal Mesh
    Socket Sampling Mode=::NewEnumerator5
   )
   &SkeletalMeshLocation(
    Bone / Socket Index=2
    MaskSpawnGroup=false
    Mesh Sampling Type=::NewEnumerator5
    Orientation Sampling=::NewEnumerator1
    Position Sampling=::NewEnumerator2
    Sampled Position Offset=<5 45 -12>
    Skeletal Mesh=$System.Skeletal Mesh
    Socket Sampling Mode=::NewEnumerator5
   ) {module=SkeletalMeshLocation001}
   /SetVariables(Particles.LightPosition:NiagaraPosition=$Particles.SkeletalMeshLocation001.SampledPosition) {isSetParameters=true}
   &FN_Light_Attributes(
    Color=<rgba 1 0.0987 0 1>
    Create Color Attribute=true
    Create Exponent Attribute=true
    Create Position Attribute=true
    Create Radius Attribute=true
    Exponent=4
    Position=&AddVectorToPosition(Position=$Particles.Position, Vector=<0 -42.5 15>)
    Radius=60
    Scale Color=&FloatFromCurve(
     FloatCurve=DataInterface() {propertyValues {Curve {
        defaultValue=1
        keys [
         {arriveTangent=9.1353, interpMode=RCIM_Cubic, leaveTangent=9.1353, tangentMode=RCTM_User, time=0, value=0}
         {interpMode=RCIM_Cubic, tangentMode=RCTM_User, time=0.2, value=1}
        ]
       }}}
     CurveIndex=$Emitter.Age
     Scale Curve=&FN_FrontEndFloatModifier(Gameplay=5, Front End=0)
    )
   )
   &FN_Light_Attributes(
    Color=<rgba 1 0.0987 0 1>
    Create Color Attribute=true
    Create Exponent Attribute=true
    Create Position Attribute=true
    Create Radius Attribute=true
    Exponent=4
    Position=&AddVectorToPosition(Position=$Particles.LightPosition, Vector=<0 30 15>)
    Radius=60
    Scale Color=&FloatFromCurve(
     FloatCurve=DataInterface() {propertyValues {Curve {
        defaultValue=1
        keys [
         {arriveTangent=9.1353, interpMode=RCIM_Cubic, leaveTangent=9.1353, tangentMode=RCTM_User, time=0, value=0}
         {interpMode=RCIM_Cubic, tangentMode=RCTM_User, time=0.2, value=1}
        ]
       }}}
     CurveIndex=$Emitter.Age
     Scale Curve=&FN_FrontEndFloatModifier(Gameplay=5, Front End=0)
    )
   ) {module=FN_Light_Attributes001}
   &FN_Light_Attributes(
    Color=<rgba 1 0.3752 0.0521 1>
    Create Color Attribute=true
    Create Exponent Attribute=true
    Create Position Attribute=true
    Create Radius Attribute=true
    Exponent=4
    Position=&AddVectorToPosition(Position=$Particles.LightPosition, Vector=<-2 -75 15>)
    Radius=55
    Scale Color=&FloatFromCurve(
     FloatCurve=DataInterface() {propertyValues {Curve {
        defaultValue=1
        keys [
         {arriveTangent=9.1353, interpMode=RCIM_Cubic, leaveTangent=9.1353, tangentMode=RCTM_User, time=0, value=0}
         {interpMode=RCIM_Cubic, tangentMode=RCTM_User, time=0.2, value=1}
        ]
       }}}
     CurveIndex=$Emitter.Age
     Scale Curve=&FN_FrontEndFloatModifier(Gameplay=&Lerp_Float(A=0, B=6, Alpha=$System.Timeline001.NormalizedTime), Front End=0)
    )
   ) {module=FN_Light_Attributes002}
   &self:InterpolateSpeedThresholdParticles(
    Define Initial Value=false
    Rate Of Change=12
    Target Float Value:NiagaraFloat=&MakeCustomFloatFromBool(
     Bool=&SetBoolByFloatComparison(
      A=&Add_Float(
       A=&Add_Float(
        A=&Abs_Float(Float:NiagaraFloat=&MakeFloatFromVector(Vector=$Engine.Owner.Velocity, Channel=::NewEnumerator0))
        B=&Abs_Float(Float:NiagaraFloat=&MakeFloatFromVector(Vector=$Engine.Owner.Velocity, Channel=::NewEnumerator2))
       )
       B=&Abs_Float(Float:NiagaraFloat=&MakeFloatFromVector(Vector=$Engine.Owner.Velocity, Channel=::NewEnumerator1))
      )
      B=150
     )
     True Float=1
     False Float=0
    )
   ) {module=InterpolateSpeedThresholdParticles}
   &ScaleColor(
    Scale Alpha=&Multiply_Float(
     A=&FN_FrontEndFloatModifier(
      Gameplay=&Lerp_Float(A=0, B=1, Alpha=$Particles.InterpolateSpeedThresholdParticles.InterpolateOverTime_Float.Moving Average)
      Front End=1
     )
     B=&FloatFromCurve(
      FloatCurve=DataInterface() {propertyValues {Curve {
         defaultValue=1
         keys [{arriveTangent=4.1785, interpMode=RCIM_Cubic, time=0, value=0}, {interpMode=RCIM_Cubic, leaveTangent=0.0224, time=1, value=1}]
        }}}
      CurveIndex=$Emitter.Age
     )
    )
    Scale RGB=&FN_FrontEndVector3Modifier(
     Gameplay=&VectorFromFloat(Value=&Lerp_Float(A=5, B=3, Alpha=$Particles.InterpolateSpeedThresholdParticles.InterpolateOverTime_Float.Moving Average))
     Front End=&VectorFromFloat(Value=3)
    )
    ScaleA=true
    ScaleRGB=true
   )
   &DynamicMaterialParameters(
    Index 0 Param 1=&FN_FrontEndFloatModifier(
     Gameplay=&FloatFromCurve(
      FloatCurve=DataInterface() {propertyValues {Curve {keys [<at 0.5 0 RCIM_Cubic>, {arriveTangent=5.7318, interpMode=RCIM_Cubic, leaveTangent=5.7318, tangentMode=RCTM_User, time=1, value=1}]}}}
      CurveIndex=$Particles.InterpolateSpeedThresholdParticles.InterpolateOverTime_Float.Moving Average
     )
     Front End=1
    )
    Module.Index0Float0Write=true
    Module.Index0Float1Write=true
    Module.Index0Float2Write=true
    Module.Index0Float3Write=true
   )
   &SolveForcesAndVelocity(Clamp Velocity=false, Limit Acceleration=false)
   /SetVariables(
    Particles.BlurPosition:NiagaraPosition=&AddVectorToPosition(Position=$Particles.Position, Vector=<0 -40 20>)
    Particles.Color2=&MakeLinearColorFromVectorAndFloat(
     "Vector (RGB)"=&MakeVectorFromLinearColorRGB(Color=&RandomRangeLinearColor(Minimum=<rgba 1 0.0312 0.0312 0.2>, Maximum=<rgba 1 0.241 0.0661 0.2>))
     "Float (Alpha)"=&FloatFromCurve(FloatCurve=DataInterface() {propertyValues {Curve {
         defaultValue=0
         keys [
          {interpMode=RCIM_Cubic, tangentMode=RCTM_User, time=0, value=0}
          {interpMode=RCIM_Cubic, tangentMode=RCTM_User, time=0.5, value=1}
          {tangentMode=RCTM_User, time=1, value=0}
         ]
        }}})
    )
    Particles.DP2:Vector4f=&MakeVector4(X=&RandomRangeFloat(Minimum=0, Maximum=1), Y=0, Z=0, W=0)
   ) {isSetParameters=true}
  ]
  rendererClasses [
   &NiagaraMeshRendererProperties
   &NiagaraSpriteRendererProperties
   &NiagaraLightRendererProperties
  ]
  renderers [
   {rendererClass=&NiagaraMeshRendererProperties, rendererIndex=0}
   {rendererClass=&NiagaraSpriteRendererProperties, rendererIndex=1}
   {rendererClass=&NiagaraLightRendererProperties, rendererIndex=2}
   {rendererClass=&NiagaraLightRendererProperties, rendererIndex=3}
  ]
 }
}
summary {
 systemName=NS_Emote_Yuzucrank_Pita_Trails
 emitters [
  {
   emitterName=RearWheelEmissive
   rendererClasses [
    &NiagaraMeshRendererProperties
    &NiagaraSpriteRendererProperties
    &NiagaraLightRendererProperties
   ]
  }
  {emitterName=FN_AnimTrail, rendererClasses [&NiagaraRibbonRendererProperties]}
  {emitterName=FN_AnimTrail001, rendererClasses [&NiagaraRibbonRendererProperties]}
  {emitterName=FN_AnimTrail002, rendererClasses [&NiagaraRibbonRendererProperties]}
 ]
 userVariables [
  User.bIsFrontEnd {type=NiagaraBool}
  User.bIsFrontEndPreview {type=NiagaraBool}
  User.Emote_Yuzucrank_Pita_RibbonWidth {type=NiagaraFloat, defaultValue=1, description=""}
  User.Emote_Yuzucrank_Pita_Z {type=NiagaraFloat, defaultValue=0, description=""}
  User.OwnerLocalVelocity {type=Vector3f, description="", defaultValue {struct=Vector3f, value=<0 0 0>}}
  User.RibbonIDOffset {type=NiagaraInt32}
 ]
}
